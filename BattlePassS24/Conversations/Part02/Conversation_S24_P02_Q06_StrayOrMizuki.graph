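# Conversation database for Season 24 narrative quest Part 2 Q06, giving main-questline dialogue with either Stray or Mizuki about the Fox Clan and River Guard alliance.
# Each NPC has its own entry point and two-line speech chain, auto-selected only while the quest is active, its objective is incomplete and the NPC carries the matching entry tag.
# Both branches converge on one shared task node that completes objective obj0, so talking to either character suffices.
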ConversationDatabase /BattlePassS24/Conversations/Part02/Conversation_S24_P02_Q06_StrayOrMizuki
entry_points [
 {tag=Conversation.Entry.Season24.Narrative.P02.Q06.Stray, destinations [#1]}
 {tag=Conversation.Entry.Season24.Narrative.P02.Q06.Mizuki, destinations [#2]}
]
nodes [
 ConversationEntryPointNode_1 {
  id=#2
  class=ConversationEntryPointNode
  outputs [#3]
  properties {entryTag {tagName=Conversation.Entry.Season24.Narrative.P02.Q06.Mizuki}}
 }
 Speech_C_1 {
  id=#3
  class=Speech_C
  outputs [#4]
  properties {
   softTaskIcon=&T_UI_NPC_MainQuest
   generalConfig {defaultMessage=I am impressed that you convinced Fox Clan to meet with us.}
  }
  sub_nodes [
   ConversationChoiceNode_4 {
    class=ConversationChoiceNode
    properties {
     defaultChoiceDisplayText=Talk to Mizuki
     choiceTags {gameplayTags [{tagName=Conversation.Choice.AutoSelectIfValid}, {tagName=Conversation.Choice.Category.MainQuestline}]}
    }
   }
   FortConversationRequirement_ControllerRequirement_12 {
    class=FortConversationRequirement_ControllerRequirement
    properties {
     nodeName=Require: Has Completed Quest
     requirement {
      class=FortControllerRequirement_HasCompletedQuest
      properties {bInverseCondition=true, questDefinition=/BattlePassS24/Items/QuestItems/NarrativeQuests/Part02/quest_s24_narrative_p02_q06}
     }
    }
   }
   FortConversationRequirement_ControllerRequirement_13 {
    class=FortConversationRequirement_ControllerRequirement
    properties {
     nodeName=Require: Is Quest Active
     requirement {
      class=FortControllerRequirement_IsQuestActive
      properties {questDefinition=&quest_s24_narrative_p02_q06}
     }
    }
   }
   FortConversationRequirement_ControllerRequirement_14 {
    class=FortConversationRequirement_ControllerRequirement
    properties {
     nodeName=Require: Has Completed Quest Objective
     requirement {
      class=FortControllerRequirement_HasCompletedQuestObjective
      properties {
       bInverseCondition=true
       objectiveBackendName=quest_s24_narrative_p02_q06_obj0
       questDefinition=&quest_s24_narrative_p02_q06
      }
     }
    }
   }
   FortConversationRequirement_ControllerRequirement_16 {
    class=FortConversationRequirement_ControllerRequirement
    properties {
     nodeName=Require: NPC Has Conversation Tag
     participantId {tagName=Conversation.Participant.NPC}
     requirement {class=FortControllerRequirement_NPCHasConversationEntryTag, properties {queryTag {tagName=Conversation.Entry.Mizuki}}}
    }
   }
  ]
 }
 Speech_C_2 {
  id=#4
  class=Speech_C
  outputs [#5]
  properties {generalConfig {defaultMessage=The River Guard accepts the Peace Syndicate's offer of alliance.}}
  sub_nodes [ConversationChoiceNode_0 {class=ConversationChoiceNode, properties {defaultChoiceDisplayText=Continue}}]
 }
 FortConversationTaskNode_CompleteQuestObjectiveByBackendName_2 {
  id=#5
  class=FortConversationTaskNode_CompleteQuestObjectiveByBackendName
  properties {
   objectiveBackendName=quest_s24_narrative_p02_q06_obj0
   questDefinition=&quest_s24_narrative_p02_q06
  }
 }
 ConversationEntryPointNode_0 {
  id=#1
  class=ConversationEntryPointNode
  outputs [#6]
  properties {entryTag {tagName=Conversation.Entry.Season24.Narrative.P02.Q06.Stray}}
 }
 Speech_C_3 {
  id=#6
  class=Speech_C
  outputs [#7]
  properties {
   softTaskIcon=&T_UI_NPC_MainQuest
   generalConfig {defaultMessage=Peace between Fox Clan and the River Guard. Never thought I'd see it.}
  }
  sub_nodes [
   ConversationChoiceNode_1 {
    class=ConversationChoiceNode
    properties {
     defaultChoiceDisplayText=Talk to Stray
     choiceTags {gameplayTags [{tagName=Conversation.Choice.AutoSelectIfValid}, {tagName=Conversation.Choice.Category.MainQuestline}]}
    }
   }
   FortConversationRequirement_ControllerRequirement_0 {
    class=FortConversationRequirement_ControllerRequirement
    properties {
     nodeName=Require: Has Completed Quest
     requirement {
      class=FortControllerRequirement_HasCompletedQuest
      properties {bInverseCondition=true, questDefinition=&quest_s24_narrative_p02_q06}
     }
    }
   }
   FortConversationRequirement_ControllerRequirement_1 {
    class=FortConversationRequirement_ControllerRequirement
    properties {
     nodeName=Require: Is Quest Active
     requirement {
      class=FortControllerRequirement_IsQuestActive
      properties {questDefinition=&quest_s24_narrative_p02_q06}
     }
    }
   }
   FortConversationRequirement_ControllerRequirement_2 {
    class=FortConversationRequirement_ControllerRequirement
    properties {
     nodeName=Require: Has Completed Quest Objective
     requirement {
      class=FortControllerRequirement_HasCompletedQuestObjective
      properties {
       bInverseCondition=true
       objectiveBackendName=quest_s24_narrative_p02_q06_obj0
       questDefinition=&quest_s24_narrative_p02_q06
      }
     }
    }
   }
   FortConversationRequirement_ControllerRequirement_3 {
    class=FortConversationRequirement_ControllerRequirement
    properties {
     nodeName=Require: NPC Has Conversation Tag
     participantId {tagName=Conversation.Participant.NPC}
     requirement {class=FortControllerRequirement_NPCHasConversationEntryTag, properties {queryTag {tagName=Conversation.Entry.Stray}}}
    }
   }
  ]
 }
 Speech_C_28 {
  id=#7
  class=Speech_C
  outputs [#5]
  properties {generalConfig {defaultMessage=We'll stand beside you in your hunt for this saboteur.}}
  sub_nodes [ConversationChoiceNode_2 {class=ConversationChoiceNode, properties {defaultChoiceDisplayText=Continue}}]
 }
]
